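# Control Rig for the face accessory of the male StuffRasp hood-down body, driving corrective motion on hood, lace and bow bones.
# Weighted parent constraints attach hood and lace bones to body bones.
# A head spherical pose reader drives bow offsets and, curved and added to remapped head-to-hood distance, reshapes and shifts the hood.
# Summed clavicle pose readers further stretch the hood; one extra hood transform node sits unconnected.
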ControlRigBlueprint /BRCosmetics/Characters/Player/Male/Medium/Bodies/M_MED_StuffRasp_HoodDown/Meshes/Parts/M_MED_StuffRasp_HoodDown_FaceAcc_CtrlRig
graphs [
 RigVMFunctionLibrary {links [], nodes []}
 RigVMModel {nodes [
   RigUnit_BeginExecution = &RigUnit_BeginExecution()
   &RigUnit_ParentConstraint(
    AdvancedSettings=(RotationOrderForFilter=XZY)
    Child=(Type=Bone,Name="deform_lacer_l")
    Parents="((Item=(Type=Bone,Name=\"deform_upperarm_up_l\"),Weight=0.0022),(Item=(Type=Bone,Name=\"deform_upperarm_fr_l\"),Weight=0.0034),(Item=(Type=Bone,Name=\"upperarm_twist_01_l\"),Weight=0.0037),(Item=(Type=Bone,Name=\"spine_03\"),Weight=0.0084),(Item=(Type=Bone,Name=\"d ...(truncated)"
    Weight=1
    ExecutePin=@RigUnit_BeginExecution.ExecutePin
   )
   > &RigUnit_ParentConstraint(
    AdvancedSettings=(RotationOrderForFilter=XZY)
    Child=(Type=Bone,Name="deform_racer_r")
    Parents="((Item=(Type=Bone,Name=\"spine_03\"),Weight=0.0031),(Item=(Type=Bone,Name=\"deform_upperarm_fr_r\"),Weight=0.0041),(Item=(Type=Bone,Name=\"deform_upperarm_up_r\"),Weight=0.0056),(Item=(Type=Bone,Name=\"upperarm_twist_01_r\"),Weight=0.006),(Item=(Type=Bone,Name=\"de ...(truncated)"
    Weight=1
   )
   > &RigUnit_ParentConstraint(
    AdvancedSettings=(RotationOrderForFilter=XZY)
    Child=(Type=Bone,Name="deform_hood")
    Parents="((Item=(Type=Bone,Name=\"upperarm_twist_01_r\"),Weight=0.0009),(Item=(Type=Bone,Name=\"neck_02\"),Weight=0.0019),(Item=(Type=Bone,Name=\"dyn_hood\"),Weight=0.0319),(Item=(Type=Bone,Name=\"clavicle_l\"),Weight=0.0523),(Item=(Type=Bone,Name=\"clavicle_r\"),Weight=0.05 ...(truncated)"
    Weight=1
   )
   > n1 = &RigUnit_SphericalPoseReader(
    ActiveRegionScaleFactors=(PositiveWidth=1,NegativeWidth=1,PositiveHeight=1,NegativeHeight=1)
    ActiveRegionSize=0.2
    Debug=(bDrawDebug=true,DebugScale=25,DebugSegments=20,DebugThickness=0.25)
    DriverAxis=(X=1,Y=0,Z=0)
    DriverItem=(Type=Bone,Name="head")
    FalloffRegionScaleFactors=(PositiveWidth=1.8,NegativeWidth=1,PositiveHeight=1,NegativeHeight=1)
    FalloffSize=0.6
    OptionalParentItem=(Type=Bone)
    RotationOffset=(X=-90,Y=0,Z=-30)
   )
   RigUnit_GetTransform = &RigUnit_GetTransform(Item=(Type=Control,Name="head_ctrl"), bInitial=False)
   RigVMFunction_MathVectorDistance = &RigVMFunction_MathVectorDistance(
    A=@RigUnit_GetTransform.Transform.Translation
    B=&RigUnit_GetTransform(Item=(Type=Bone,Name="deform_hood"), bInitial=False).Transform.Translation
   )
   Remap = &RigVMFunction_MathDoubleRemap(SourceMaximum=10, SourceMinimum=14, TargetMaximum=0.3, TargetMinimum=0, bClamp=False, Value=@RigVMFunction_MathVectorDistance.Result)
   RigVMFunction_AnimEvalRichCurve = &RigVMFunction_AnimEvalRichCurve(
    Curve=(EditorCurveData=(Keys=((InterpMode=RCIM_Cubic),(InterpMode=RCIM_Cubic,TangentMode=RCTM_User,Time=1,Value=1,ArriveTangent=1.16,LeaveTangent=1.16))))
    SourceMinimum=0
    TargetMinimum=0
    Value=@n1.OutputParam
   )
   Add = &RigVMFunction_MathDoubleAdd(B=@Remap.Result, A=@RigVMFunction_AnimEvalRichCurve.Result)
   > &RigUnit_ModifyTransforms(
    ItemToModify=((Item=(Type=Bone,Name="deform_hood"),Transform=(Rotation=(X=0,Y=0,Z=-0.1736,W=0.9848),Translation=(X=-2,Y=-4,Z=0),Scale3D=(X=0.4,Y=2,Z=1.6))))
    WeightMinimum=0
    Weight=@Add.Result
   )
   Multiply = &RigVMFunction_MathDoubleMul(B=2, A=@n1.OutputParam)
   > &RigUnit_ModifyTransforms(
    ItemToModify="((Item=(Type=Bone,Name=\"dyn_bow_l\"),Transform=(Rotation=(X=0,Y=0,Z=0.2588,W=0.9659),Translation=(X=0,Y=0,Z=0),Scale3D=(X=1,Y=1,Z=1))),(Item=(Type=Bone,Name=\"deform_bow\"),Transform=(Rotation=(X=0,Y=0,Z=0,W=1),Translation=(X=0,Y=0,Z=0),Scale3D=(X=1,Y=1,Z=1)) ...(truncated)"
    WeightMinimum=0
    Weight=@Multiply.Result
   )
   > n2 = &RigUnit_SphericalPoseReader(
    ActiveRegionScaleFactors=(PositiveWidth=1,NegativeWidth=1,PositiveHeight=1,NegativeHeight=1)
    ActiveRegionSize=0.1
    Debug=(bDrawDebug=false,DebugScale=25,DebugSegments=20,DebugThickness=0.25)
    DriverAxis=(X=-1,Y=0,Z=0)
    DriverItem=(Type=Bone,Name="clavicle_r")
    FalloffRegionScaleFactors=(PositiveWidth=1,NegativeWidth=1,PositiveHeight=1,NegativeHeight=1)
    FalloffSize=0.3
    OptionalParentItem=(Type=Bone)
    RotationOffset=(X=150,Y=0,Z=50)
   )
   > n3 = &RigUnit_SphericalPoseReader(
    ActiveRegionScaleFactors=(PositiveWidth=1,NegativeWidth=1,PositiveHeight=1,NegativeHeight=1)
    ActiveRegionSize=0.1
    Debug=(bDrawDebug=false,DebugScale=25,DebugSegments=20,DebugThickness=0.25)
    DriverAxis=(X=1,Y=0,Z=0)
    DriverItem=(Type=Bone,Name="clavicle_l")
    FalloffRegionScaleFactors=(PositiveWidth=1,NegativeWidth=1,PositiveHeight=1,NegativeHeight=1)
    FalloffSize=0.3
    OptionalParentItem=(Type=Bone)
    RotationOffset=(X=10,Y=10,Z=-160)
   )
   > &RigUnit_ModifyTransforms(
    ItemToModify=((Item=(Type=Bone,Name="deform_hood"),Transform=(Rotation=(X=0,Y=0,Z=0,W=1),Translation=(X=0,Y=0,Z=0),Scale3D=(X=1,Y=1,Z=1.6))))
    WeightMinimum=0
    Weight=&RigVMFunction_MathDoubleAdd(A=@n2.OutputParam, B=@n3.OutputParam).Result
   )
   &RigUnit_ModifyTransforms(
    ExecutePin=()
    ItemToModify=((Item=(Type=Bone,Name="deform_hood"),Transform=(Rotation=(X=0,Y=0,Z=-0.2588,W=0.9659),Translation=(X=-3,Y=-2,Z=0),Scale3D=(X=0.3,Y=2,Z=2))))
    Weight=1
    WeightMinimum=0
   )
  ]}
]
hierarchy [
 root {loc=<0 0 0>, rot=<0 0 0 1>}
 pelvis {parent=root, loc=<0 -0.7071 93.1191>, rot=<0.0127 0.707 -0.0127 -0.707>}
 spine_01 {parent=pelvis, loc=<6.7494 -0.6388 0>, rot=<0 0 0.034 -0.9994>}
 spine_02 {parent=spine_01, loc=<7.2518 -0.035 0>, rot=<0 0 0.0238 -0.9997>}
 spine_03 {parent=spine_02, loc=<6.0104 0 0>, rot=<0 0 0.0149 -0.9999>}
 spine_04 {parent=spine_03, loc=<8.1024 0 0>, rot=<0 0 -0.0132 -0.9999>}
 spine_05 {parent=spine_04, loc=<10.0146 0 0>, rot=<0 0 -0.0015 -1>}
 neck_01 {parent=spine_05, loc=<11.3138 0.1529 0>, rot=<0 0 -0.2003 -0.9797>}
 neck_02 {parent=neck_01, loc=<4.5627 0 0>, rot=<0 0 0.0504 -0.9987>}
 head {parent=neck_02, loc=<5.3914 0 0>, rot=<0 0 0.115 -0.9934>}
 head_ctrl {controlType=EULER_TRANSFORM, parent=head, type=Control, loc=<8 -6 0>, rot=<0 0 0 1>}
 FACIAL_C_FacialRoot {parent=head, loc=<2.1524 -0.746 -0.0002>, rot=<0.4879 -0.5118 -0.4879 -0.5118>}
 FACIAL_C_Forehead {parent=FACIAL_C_FacialRoot, loc=<0.0007 -3.6902 10.367>, rot=<-0.0609 -0.0002 -0.0002 -0.9981>}
 FACIAL_L_ForeheadIn {parent=FACIAL_C_FacialRoot, loc=<1.4434 -3.6526 10.1556>, rot=<-0.0781 -0.0698 0.0029 -0.9945>}
 FACIAL_R_ForeheadIn {parent=FACIAL_C_FacialRoot, loc=<-1.4431 -3.6558 10.1539>, rot=<-0.0781 0.0695 -0.0032 -0.9945>}
 FACIAL_L_ForeheadMid {parent=FACIAL_C_FacialRoot, loc=<2.8107 -3.8309 9.6815>, rot=<-0.0881 -0.1754 -0.0238 -0.9803>}
 FACIAL_R_ForeheadMid {parent=FACIAL_C_FacialRoot, loc=<-2.8097 -3.8323 9.6814>, rot=<-0.0882 0.1751 0.0235 -0.9803>}
 FACIAL_L_ForeheadOut {parent=FACIAL_C_FacialRoot, loc=<4.1018 -3.688 8.8174>, rot=<-0.138 -0.2543 -0.0838 -0.9535>}
 FACIAL_R_ForeheadOut {parent=FACIAL_C_FacialRoot, loc=<-4.1023 -3.6889 8.8136>, rot=<-0.1381 0.254 0.0835 -0.9536>}
 FACIAL_L_EyesackUpper {parent=FACIAL_C_FacialRoot, loc=<3.0752 -2.8606 10.054>, rot=<0.0315 -0.0832 -0.0662 -0.9938>}
 FACIAL_R_EyesackUpper {parent=FACIAL_C_FacialRoot, loc=<-3.0711 -2.8601 10.0584>, rot=<0.0315 0.083 0.0658 -0.9939>}
 FACIAL_L_EyelidUpperFurrow {parent=FACIAL_C_FacialRoot, loc=<2.8593 -2.6413 9.8173>, rot=<-0.0472 -0.1496 -0.0824 -0.9842>}
 FACIAL_R_EyelidUpperFurrow {parent=FACIAL_C_FacialRoot, loc=<-2.855 -2.6553 9.8106>, rot=<-0.0473 0.1493 0.0821 -0.9843>}
 FACIAL_L_EyelidUpperA {parent=FACIAL_C_FacialRoot, loc=<2.8876 -2.0622 8.3479>, rot=<-0.1651 -0.0002 -0.0001 -0.9863>}
 FACIAL_R_EyelidUpperA {parent=FACIAL_C_FacialRoot, loc=<-2.8893 -2.0616 8.347>, rot=<-0.1651 -0.0002 -0.0001 -0.9863>}
 FACIAL_L_Eye {parent=FACIAL_C_FacialRoot, loc=<2.876 -2.0648 8.3439>, rot=<-0.0175 -0.0002 -0.0002 -0.9998>}
 FACIAL_R_Eye {parent=FACIAL_C_FacialRoot, loc=<-2.8784 -2.0658 8.3431>, rot=<-0.0175 -0.0002 -0.0002 -0.9998>}
 FACIAL_L_EyelidLowerA {parent=FACIAL_C_FacialRoot, loc=<2.8876 -2.0622 8.3479>, rot=<0.2672 -0.0001 -0.0002 -0.9636>}
 FACIAL_R_EyelidLowerA {parent=FACIAL_C_FacialRoot, loc=<-2.8893 -2.0616 8.347>, rot=<0.2672 -0.0001 -0.0002 -0.9636>}
 FACIAL_L_EyeCornerInner {parent=FACIAL_C_FacialRoot, loc=<1.814 -1.5837 9.1505>, rot=<-0.0175 -0.0002 -0.0002 -0.9998>}
 FACIAL_R_EyeCornerInner {parent=FACIAL_C_FacialRoot, loc=<-1.8217 -1.5849 9.1419>, rot=<-0.0175 -0.0002 -0.0002 -0.9998>}
 FACIAL_L_EyeCornerOuter {parent=FACIAL_C_FacialRoot, loc=<4.2337 -1.7065 8.6322>, rot=<-0.0161 -0.3827 -0.0068 -0.9237>}
 FACIAL_R_EyeCornerOuter {parent=FACIAL_C_FacialRoot, loc=<-4.2336 -1.7089 8.6267>, rot=<-0.0162 0.3824 0.0065 -0.9238>}
 FACIAL_L_EyesackLower {parent=FACIAL_C_FacialRoot, loc=<3.0549 -0.6086 8.5738>, rot=<-0.0395 -0.1394 -0.0249 -0.9891>}
 FACIAL_R_EyesackLower {parent=FACIAL_C_FacialRoot, loc=<-3.0558 -0.6132 8.5635>, rot=<-0.0396 0.1391 0.0245 -0.9892>}
 FACIAL_L_CheekInner {parent=FACIAL_C_FacialRoot, loc=<2.0032 0.1481 8.9026>, rot=<-0.1401 -0.0847 -0.1988 -0.9663>}
 FACIAL_R_CheekInner {parent=FACIAL_C_FacialRoot, loc=<-2.0029 0.1519 8.9028>, rot=<-0.1401 0.0844 0.1985 -0.9664>}
 FACIAL_L_CheekOuter {parent=FACIAL_C_FacialRoot, loc=<3.7239 0.3384 7.1002>, rot=<0.0392 -0.3788 0.0936 -0.9199>}
 FACIAL_R_CheekOuter {parent=FACIAL_C_FacialRoot, loc=<-3.7208 0.3486 7.096>, rot=<0.039 0.3785 -0.0939 -0.92>}
 FACIAL_L_NasolabialBulge {parent=FACIAL_C_FacialRoot, loc=<3.902 3.3083 6.9264>, rot=<0.0418 -0.3778 0.017 -0.9248>}
 FACIAL_R_NasolabialBulge {parent=FACIAL_C_FacialRoot, loc=<-3.8934 3.3443 6.92>, rot=<0.0417 0.3776 -0.0174 -0.9249>}
 FACIAL_L_NasolabialFurrow {parent=FACIAL_C_FacialRoot, loc=<2.7538 2.3995 9.6495>, rot=<-0.0494 -0.1196 -0.3659 -0.9216>}
 FACIAL_R_NasolabialFurrow {parent=FACIAL_C_FacialRoot, loc=<-2.749 2.4192 9.6495>, rot=<-0.0494 0.1193 0.3656 -0.9218>}
 FACIAL_L_CheekLower {parent=FACIAL_C_FacialRoot, loc=<4.2502 3.3019 6.3245>, rot=<0.0189 -0.428 0.0501 -0.9022>}
 FACIAL_R_CheekLower {parent=FACIAL_C_FacialRoot, loc=<-4.2474 3.3299 6.3194>, rot=<0.0187 0.4277 -0.0504 -0.9023>}
 FACIAL_C_Nose {parent=FACIAL_C_FacialRoot, loc=<-0.0005 -0.2183 10.8864>, rot=<-0.1476 -0.0002 -0.0002 -0.989>}
 FACIAL_C_NoseLower {parent=FACIAL_C_Nose, loc=<0.002 3.0206 -0.5892>, rot=<0.1305 0 0 -0.9914>}
 FACIAL_L_Nostril {parent=FACIAL_C_Nose, loc=<0.9808 2.1397 -0.5376>, rot=<0.0173 0.1305 0.0023 -0.9913>}
 FACIAL_R_Nostril {parent=FACIAL_C_Nose, loc=<-0.9934 2.136 -0.5405>, rot=<0.0173 -0.1305 -0.0023 -0.9913>}
 FACIAL_C_NoseTip {parent=FACIAL_C_Nose, loc=<0 2.1705 0.3969>, rot=<0.0872 0 0 -0.9962>}
 FACIAL_C_MouthUpper {parent=FACIAL_C_FacialRoot, loc=<0.0043 3.8691 7.9129>, rot=<-0.0174 -0.0002 -0.0002 -0.9998>}
 FACIAL_C_LipUpper {parent=FACIAL_C_MouthUpper, loc=<0.0021 0.1513 2.7227>, rot=<-0.0872 0 0 -0.9962>}
 FACIAL_L_LipUpper {parent=FACIAL_C_MouthUpper, loc=<1.1686 0.2778 2.481>, rot=<-0.0778 -0.1601 -0.0379 -0.9833>}
 FACIAL_R_LipUpper {parent=FACIAL_C_MouthUpper, loc=<-1.1668 0.2859 2.479>, rot=<-0.0778 0.1601 0.0379 -0.9833>}
 FACIAL_L_LipUpperOuter {parent=FACIAL_C_MouthUpper, loc=<1.78 0.5053 1.993>, rot=<-0.0984 -0.2293 -0.0983 -0.9634>}
 FACIAL_R_LipUpperOuter {parent=FACIAL_C_MouthUpper, loc=<-1.7915 0.5148 1.9884>, rot=<-0.0984 0.2293 0.0983 -0.9634>}
 FACIAL_L_LipCorner {parent=FACIAL_C_MouthUpper, loc=<2.2278 0.4833 1.1506>, rot=<0 -0.2078 0 -0.9782>}
 FACIAL_R_LipCorner {parent=FACIAL_C_MouthUpper, loc=<-2.1127 0.6277 1.38>, rot=<0 0.2078 0 -0.9782>}
 FACIAL_L_Masseter {parent=FACIAL_C_FacialRoot, loc=<6.2184 1.0121 4.146>, rot=<0.052 0.1126 0.0097 -0.9922>}
 FACIAL_R_Masseter {parent=FACIAL_C_FacialRoot, loc=<-6.2265 1.0164 4.1248>, rot=<0.052 -0.1129 -0.01 -0.9922>}
 FACIAL_C_UnderChin {parent=FACIAL_C_FacialRoot, loc=<-0.0006 8.8098 5.1691>, rot=<0.4847 0 -0.0002 -0.8747>}
 FACIAL_L_UnderChin {parent=FACIAL_C_FacialRoot, loc=<3.2895 7.131 3.4947>, rot=<0.3879 -0.3563 0.0759 -0.8467>}
 FACIAL_R_UnderChin {parent=FACIAL_C_FacialRoot, loc=<-3.2946 7.1346 3.4962>, rot=<0.3877 0.3562 -0.0763 -0.8467>}
 FACIAL_C_TeethUpper {parent=FACIAL_C_FacialRoot, loc=<0.0049 4.5867 8.9282>, rot=<-0.0174 -0.0002 -0.0002 -0.9998>}
 FACIAL_C_LowerLipRotation {parent=FACIAL_C_FacialRoot, loc=<0.002 1.1885 2.1499>, rot=<0.3256 -0.0001 -0.0002 -0.9455>}
 FACIAL_C_MouthLower {parent=FACIAL_C_LowerLipRotation, loc=<0.0014 -1.4359 6.1917>, rot=<-0.342 0 0 -0.9397>}
 FACIAL_C_LipLower {parent=FACIAL_C_MouthLower, loc=<-0.0037 1.9453 1.8417>, rot=<0.0872 0 0 -0.9962>}
 FACIAL_L_LipLower {parent=FACIAL_C_MouthLower, loc=<1.0664 1.8868 1.6922>, rot=<0.0927 -0.1294 0.0272 -0.9869>}
 FACIAL_R_LipLower {parent=FACIAL_C_MouthLower, loc=<-1.0691 1.8895 1.71>, rot=<0.0927 0.1294 -0.0272 -0.9869>}
 FACIAL_L_LipLowerOuter {parent=FACIAL_C_MouthLower, loc=<1.9231 1.2706 1.4582>, rot=<0.0913 -0.2321 0.0318 -0.9679>}
 FACIAL_R_LipLowerOuter {parent=FACIAL_C_MouthLower, loc=<-1.9176 1.3091 1.5178>, rot=<0.0913 0.2321 -0.0318 -0.9679>}
 FACIAL_C_TeethLower {parent=FACIAL_C_LowerLipRotation, loc=<0.0018 -1.4387 7.3155>, rot=<-0.342 0 0 -0.9397>}
 FACIAL_C_Tongue1 {parent=FACIAL_C_TeethLower, loc=<-0.0053 -0.2018 -4.8428>, rot=<-0.2588 0 0 -0.9659>}
 FACIAL_C_Tongue2 {parent=FACIAL_C_Tongue1, loc=<0.0009 0.4772 1.4354>, rot=<0.2588 0 0 -0.9659>}
 FACIAL_C_Tongue3 {parent=FACIAL_C_Tongue2, loc=<0.0017 0.2305 1.1622>, rot=<0.1305 0 0 -0.9914>}
 FACIAL_C_Tongue4 {parent=FACIAL_C_Tongue3, loc=<0.0025 -0.0807 0.9915>, rot=<0.0436 0 0 -0.999>}
 FACIAL_C_TongueUpper3 {parent=FACIAL_C_Tongue3, loc=<-0.0004 -0.43 -0.0036>, rot=<0 0 0 1>}
 FACIAL_C_TongueLower3 {parent=FACIAL_C_Tongue3, loc=<-0.0026 0.8282 -0.1285>, rot=<0 0 0 1>}
 FACIAL_L_TongueSide3 {parent=FACIAL_C_Tongue3, loc=<1.8341 -0.0249 0.1356>, rot=<0 0 0 1>}
 FACIAL_R_TongueSide3 {parent=FACIAL_C_Tongue3, loc=<-1.8344 -0.0077 0.1281>, rot=<0 0 0 1>}
 FACIAL_C_TongueUpper2 {parent=FACIAL_C_Tongue2, loc=<0.0008 -0.5418 0.3373>, rot=<0 0 0 1>}
 FACIAL_L_TongueSide2 {parent=FACIAL_C_Tongue2, loc=<2.006 0.2381 0.4348>, rot=<0 0 0 1>}
 FACIAL_R_TongueSide2 {parent=FACIAL_C_Tongue2, loc=<-1.9998 0.257 0.4303>, rot=<0 0 0 1>}
 FACIAL_C_Jaw {parent=FACIAL_C_FacialRoot, loc=<0.002 1.1885 2.1499>, rot=<0.3256 -0.0001 -0.0002 -0.9455>}
 FACIAL_C_Jawline {parent=FACIAL_C_Jaw, loc=<-0.0015 2.8769 8.4497>, rot=<0.2586 0 0 -0.966>}
 FACIAL_L_ChinSide {parent=FACIAL_C_Jaw, loc=<3.0008 0.4633 7.9267>, rot=<-0.2208 -0.3738 0.0499 -0.8995>}
 FACIAL_R_ChinSide {parent=FACIAL_C_Jaw, loc=<-3.0041 0.4634 7.9357>, rot=<-0.2208 0.3738 -0.0499 -0.8995>}
 FACIAL_C_Chin {parent=FACIAL_C_Jaw, loc=<-0.002 1.3305 8.5135>, rot=<-0.2591 0 0 -0.9659>}
 FACIAL_L_Ear {parent=FACIAL_C_FacialRoot, loc=<6.2174 -0.7318 0.5387>, rot=<-0.0438 -0.0004 -0.0263 -0.9987>}
 FACIAL_R_Ear {parent=FACIAL_C_FacialRoot, loc=<-6.2175 -0.7292 0.5409>, rot=<-0.0437 0.0001 0.026 -0.9987>}
 FACIAL_L_Jawline {parent=FACIAL_C_FacialRoot, loc=<4.6636 4.8826 4.7778>, rot=<0.1924 0.1586 -0.0521 -0.967>}
 FACIAL_R_Jawline {parent=FACIAL_C_FacialRoot, loc=<-4.6788 4.8796 4.7723>, rot=<0.1925 -0.1588 0.0517 -0.967>}
 FACIAL_C_NoseBridge {parent=FACIAL_C_FacialRoot, loc=<-0.0004 -2.3116 11.277>, rot=<0.0264 -0.0001 -0.0002 -0.9997>}
 FACIAL_L_EyelidUpperB {parent=FACIAL_C_FacialRoot, loc=<2.8876 -2.0622 8.3479>, rot=<-0.2419 -0.0002 -0.0001 -0.9703>}
 FACIAL_L_EyelidLowerB {parent=FACIAL_C_FacialRoot, loc=<2.8876 -2.0622 8.3479>, rot=<0.3007 -0.0001 -0.0002 -0.9537>}
 FACIAL_R_EyelidLowerB {parent=FACIAL_C_FacialRoot, loc=<-2.8893 -2.0616 8.347>, rot=<0.309 -0.0001 -0.0002 -0.9511>}
 FACIAL_R_EyelidUpperB {parent=FACIAL_C_FacialRoot, loc=<-2.8893 -2.0616 8.347>, rot=<-0.242 -0.0002 -0.0001 -0.9703>}
 deform_bow {parent=head, loc=<4.1734 -6.0621 0>, rot=<0 0 0 1>}
 dyn_bow_l {parent=deform_bow, loc=<1.0383 -0.913 -1.6201>, rot=<0.0846 -0.7009 0.0901 -0.7025>}
 dyn_bow_r {parent=deform_bow, loc=<-1.0071 -0.421 0.1879>, rot=<0.1112 -0.9474 0.1119 0.2784>}
 clavicle_l {parent=spine_05, loc=<4.3575 -0.1466 -2.3223>, rot=<0.0465 -0.7289 -0.009 -0.6829>}
 upperarm_l {parent=clavicle_l, loc=<15.5111 0 0>, rot=<0.0226 -0.3234 -0.0088 -0.9459>}
 upperarm_twist_01_l {parent=upperarm_l, loc=<0.0041 0 0>, rot=<0 0 0 -1>}
 deform_upperarm_bk_base_l {parent=upperarm_twist_01_l, loc=<1.0813 -7.2193 -1.9453>, rot=<0 0 0 1>}
 deform_upperarm_bk_l {parent=deform_upperarm_bk_base_l, loc=<0 0 0>, rot=<0 0 0 1>}
 deform_upperarm_fr_base_l {parent=upperarm_twist_01_l, loc=<-0.3691 7.0506 -2.765>, rot=<0 0 0 1>}
 deform_upperarm_fr_l {parent=deform_upperarm_fr_base_l, loc=<0 0 0>, rot=<0 0 0 1>}
 deform_upperarm_up_base_l {parent=upperarm_twist_01_l, loc=<-3.2216 -1.1391 3.3311>, rot=<0 0 0 1>}
 deform_upperarm_up_l {parent=deform_upperarm_up_base_l, loc=<0 0 0>, rot=<0 0 0 1>}
 clavicle_r {parent=spine_05, loc=<4.3575 -0.1466 2.3223>, rot=<0.7289 0.0465 -0.6829 0.009>}
 upperarm_r {parent=clavicle_r, loc=<-15.5111 0 0>, rot=<0.0226 -0.3234 -0.0088 -0.9459>}
 upperarm_twist_01_r {parent=upperarm_r, loc=<-0.0041 0 0>, rot=<0 0 0 -1>}
 deform_upperarm_bk_base_r {parent=upperarm_twist_01_r, loc=<-1.0814 7.2192 1.9455>, rot=<0 0 0 1>}
 deform_upperarm_bk_r {parent=deform_upperarm_bk_base_r, loc=<0 0 0>, rot=<0 0 0 1>}
 deform_upperarm_fr_base_r {parent=upperarm_twist_01_r, loc=<0.3689 -7.0506 2.7652>, rot=<0 0 0 1>}
 deform_upperarm_fr_r {parent=deform_upperarm_fr_base_r, loc=<0 0 0>, rot=<0 0 0 1>}
 deform_upperarm_up_base_r {parent=upperarm_twist_01_r, loc=<3.2214 1.1391 -3.3308>, rot=<0 0 0 1>}
 deform_upperarm_up_r {parent=deform_upperarm_up_base_r, loc=<0 0 0>, rot=<0 0 0 -1>}
 deform_lacer_l {parent=spine_05, loc=<8.666 8.4133 -5.2729>, rot=<-0.0142 0.0564 0.9817 0.1811>}
 dyn_lacer_l {parent=deform_lacer_l, loc=<0 0 0>, rot=<0 0 0 1>}
 dyn_lacer_1_l {parent=dyn_lacer_l, loc=<4.5627 0 0>, rot=<0 0 0.0072 -1>}
 deform_racer_r {parent=spine_05, loc=<8.6657 8.4133 5.2729>, rot=<-0.0564 -0.0142 0.1811 -0.9817>}
 dyn_racer_r {parent=deform_racer_r, loc=<0 0 0>, rot=<0 0 0 1>}
 dyn_racer_1_r {parent=dyn_racer_r, loc=<-4.5624 0.0001 0>, rot=<0 0 0.0072 -1>}
 deform_hood {parent=spine_05, loc=<8.7202 -9.6847 0>, rot=<0 0 -0.1109 -0.9938>}
 dyn_hood {parent=deform_hood, loc=<0 0 0>, rot=<0 0 0 1>}
 deform_pec_aim_base_l {parent=spine_04, loc=<7.8732 12.6388 -2.4123>, rot=<0.0102 -0.4602 0.3979 -0.7936>}
 deform_pec_aim_l {parent=deform_pec_aim_base_l, loc=<0 0 0>, rot=<0 0 0 -1>}
 deform_pec_base_l {parent=deform_pec_aim_l, loc=<10.766 0 0>, rot=<0 0 0 1>}
 deform_pec_l {parent=deform_pec_base_l, loc=<0.055 0 0>, rot=<0 0 0 1>}
 deform_pec_aim_base_r {parent=spine_04, loc=<7.8727 12.6388 2.4123>, rot=<-0.4602 -0.0102 0.7936 0.3979>}
 deform_pec_aim_r {parent=deform_pec_aim_base_r, loc=<0 0 0>, rot=<0 0 0 -1>}
 deform_pec_base_r {parent=deform_pec_aim_r, loc=<-10.7662 -0.0006 -0.0007>, rot=<0 0 0 1>}
 deform_pec_r {parent=deform_pec_base_r, loc=<0 0 0>, rot=<0 0 0 1>}
]
