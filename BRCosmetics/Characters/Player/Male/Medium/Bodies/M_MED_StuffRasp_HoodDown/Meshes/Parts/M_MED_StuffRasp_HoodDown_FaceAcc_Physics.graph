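# Physics asset for the FaceAcc part of the M_MED_StuffRasp_HoodDown male body, built from capsules on upper-spine, pectoral and lacer/racer bones.
# Two short dyn bone chains, left and right, hang from deform lacer/racer anchors on spine_05 with tight swing and twist limits.
# Spine and pectoral capsules form the torso volume, linked by free angular constraints, as are the anchors to spine_05.
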PhysicsAsset /BRCosmetics/Characters/Player/Male/Medium/Bodies/M_MED_StuffRasp_HoodDown/Meshes/Parts/M_MED_StuffRasp_HoodDown_FaceAcc_Physics
bodies {
 deform_lacer_l [{shapeType=Capsule, length=0.505, radius=0.505, shapeName=deform_lacer_l_capsule, rotation=<rot 0 -90 0>}]
 deform_pec_base_l [{shapeType=Capsule, length=14.25, radius=5.75, shapeName=deform_pec_base_l_capsule, center=<-3.2301 -3.7893 0>, rotation=<rot 90 10.0928 0>}]
 deform_pec_base_r [{shapeType=Capsule, length=14.25, radius=5.75, shapeName=deform_pec_base_r_capsule, center=<3.2524 3.6641 0>, rotation=<rot 90 -169.9072 0>}]
 deform_pec_l [{shapeType=Capsule, length=14.75, radius=5.25, shapeName=deform_pec_l_capsule, center=<0.4555 -2.4962 0>, rotation=<rot 90 -165.5426 0>}]
 deform_pec_r [{shapeType=Capsule, length=14.75, radius=5.25, shapeName=deform_pec_r_capsule, center=<-0.4555 2.4962 0>, rotation=<rot 90 14.4574 0>}]
 deform_racer_r [{shapeType=Capsule, length=0.505, radius=0.505, shapeName=deform_racer_r_capsule, rotation=<rot 0 -90 0>}]
 dyn_lacer_1_l [{shapeType=Capsule, length=7, radius=1, shapeName=dyn_lacer_1_l_capsule, center=<3.457 -0.3463 0>, rotation=<rot 90 0 0>}]
 dyn_lacer_l [{shapeType=Capsule, length=4.5, radius=1, shapeName=dyn_lacer_l_capsule, center=<2.236 -0.3463 0>, rotation=<rot 90 0 0>}]
 dyn_racer_1_r [{shapeType=Capsule, length=7.5, radius=1, shapeName=dyn_racer_1_r_capsule, center=<-3.8119 0.5274 0>, rotation=<rot 90 0 0>}]
 dyn_racer_r [{shapeType=Capsule, length=4.75, radius=1, shapeName=dyn_racer_r_capsule, center=<-2.3626 0.5274 0>, rotation=<rot 90 0 0>}]
 spine_04 [{shapeType=Capsule, length=8.5, radius=11.5, shapeName=spine_04_capsule, center=<6.0766 0 0>}]
 spine_05 [{shapeType=Capsule, length=18.25, radius=8.75, shapeName=spine_05_capsule, center=<3.8095 0 0>}]
}
constraints [
 {
  bone1Name=spine_05
  bone2Name=spine_04
  swing1LimitDegrees=45
  swing1Motion=Free
  swing2LimitDegrees=45
  swing2Motion=Free
  twistLimitDegrees=45
  twistMotion=Free
 }
 {bone1Name=dyn_lacer_l, bone2Name=deform_lacer_l, swing1LimitDegrees=15, swing2LimitDegrees=15, twistLimitDegrees=5}
 {bone1Name=dyn_lacer_1_l, bone2Name=dyn_lacer_l, swing1LimitDegrees=15, swing2LimitDegrees=15, twistLimitDegrees=5}
 {bone1Name=dyn_racer_r, bone2Name=deform_racer_r, swing1LimitDegrees=15, swing2LimitDegrees=15, twistLimitDegrees=5}
 {bone1Name=dyn_racer_1_r, bone2Name=dyn_racer_r, swing1LimitDegrees=15, swing2LimitDegrees=15, twistLimitDegrees=5}
 {
  bone1Name=deform_lacer_l
  bone2Name=spine_05
  swing1LimitDegrees=45
  swing1Motion=Free
  swing2LimitDegrees=45
  swing2Motion=Free
  twistLimitDegrees=45
  twistMotion=Free
 }
 {
  bone1Name=deform_racer_r
  bone2Name=spine_05
  swing1LimitDegrees=45
  swing1Motion=Free
  swing2LimitDegrees=45
  swing2Motion=Free
  twistLimitDegrees=45
  twistMotion=Free
 }
 {
  bone1Name=deform_pec_base_r
  bone2Name=spine_04
  swing1LimitDegrees=45
  swing1Motion=Free
  swing2LimitDegrees=45
  swing2Motion=Free
  twistLimitDegrees=45
  twistMotion=Free
 }
 {
  bone1Name=deform_pec_base_l
  bone2Name=spine_04
  swing1LimitDegrees=45
  swing1Motion=Free
  swing2LimitDegrees=45
  swing2Motion=Free
  twistLimitDegrees=45
  twistMotion=Free
 }
 {
  bone1Name=deform_pec_l
  bone2Name=deform_pec_base_l
  swing1LimitDegrees=45
  swing1Motion=Free
  swing2LimitDegrees=45
  swing2Motion=Free
  twistLimitDegrees=45
  twistMotion=Free
 }
 {
  bone1Name=deform_pec_r
  bone2Name=deform_pec_base_r
  swing1LimitDegrees=45
  swing1Motion=Free
  swing2LimitDegrees=45
  swing2Motion=Free
  twistLimitDegrees=45
  twistMotion=Free
 }
]
