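# Animation Blueprint for the face accessory mesh of the male StuffRasp HoodDown character cosmetic.
# It layers a pose copied from Body_Mesh, cached when character customization loads, over the attached parent's pose at clavicle and pectoral bones.
# The result feeds a Control Rig, then RigidBody physics simulated relative to spine_05.
# Two Anim Dynamics nodes finish with damped angular-spring motion on the left and right bow bones, relative to the head.
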AnimBlueprint /BRCosmetics/Characters/Player/Male/Medium/Bodies/M_MED_StuffRasp_HoodDown/Meshes/Parts/M_MED_StuffRasp_HoodDown_FaceAcc_AnimBP
target_skeleton=&/M_MED_StuffRasp_HoodDown_FaceAcc_Skeleton
graphs [
 AnimGraph {nodes [
   n1 = Layered blend per bone(
    BasePose=Copy Pose From Mesh() {bUseAttachedParent=true}
    BlendPoses_0=Copy Pose From Mesh(SourceMeshComponent=Get Body_Mesh()) {bUseAttachedParent=true}
   ) {
    bMeshSpaceRotationBlend=true
    layerSetup [{branchFilters [
       {blendDepth=3, boneName=clavicle_l}
       {blendDepth=3, boneName=clavicle_r}
       {blendDepth=3, boneName=deform_pec_aim_base_l}
       {blendDepth=3, boneName=deform_pec_aim_base_r}
      ]}]
   }
   n2 = Control Rig(Source=@n1) {controlRigAssetReference {blueprintRigClass=&/M_MED_StuffRasp_HoodDown_FaceAcc_CtrlRig.M_MED_StuffRasp_HoodDown_FaceAcc_CtrlRig_C}}
   n3 = RigidBody(ComponentPose=Local To Component(LocalPose=@n2)) {
    bClampLinearTranslationLimitToRefPose=true
    bOverrideWorldGravity=true
    overridePhysicsAsset=&/M_MED_StuffRasp_HoodDown_FaceAcc_Physics
    simulationSpace=BaseBoneSpace
    baseBoneRef {boneName=spine_05}
    overrideWorldGravity {z=-980}
    simSpaceSettings {
     maxAngularAcceleration=500
     maxAngularVelocity=150
     maxLinearAcceleration=1500
     maxLinearVelocity=500
     worldAlpha=0.4
     externalLinearDragV=<0.2 0.2 0.2>
    }
   }
   n4 = Anim Dynamics - Bone: dyn_bow_l!(ComponentPose=@n3, Alpha=1, GravityOverride:Vector="0,0,0") -> (Pose) {
    angularDampingOverride=0.9999
    angularSpringConstant=500
    bAngularSpring=true
    bOverrideAngularDamping=true
    simulationSpace=RootRelative
    boundBone {boneName=dyn_bow_l}
    physicsBodyDefinitions [{
      boundBone {boneName=dyn_bow_l}
      boxExtents=<10 5 5>
      constraintSetup {angularLimitsMax=<0 15 0>, angularLimitsMin=<0 -15 -10>, angularTarget=<1 0 0>}
      localJointOffset=<0.01 0 0>
     }]
    relativeSpaceBone {boneName=head}
    simSpaceSettings {maxAngularAcceleration=5000, maxAngularVelocity=400, simSpaceAngularAlpha=0.2}
   }
   n5 = Anim Dynamics - Bone: dyn_bow_r!(ComponentPose=@n4, Alpha=1, GravityOverride:Vector="0,0,0") -> (Pose) {
    angularDampingOverride=0.9999
    angularSpringConstant=500
    bAngularSpring=true
    bOverrideAngularDamping=true
    simulationSpace=RootRelative
    boundBone {boneName=dyn_bow_r}
    physicsBodyDefinitions [{
      boundBone {boneName=dyn_bow_r}
      boxExtents=<10 5 5>
      constraintSetup {angularLimitsMax=<0 15 0>, angularLimitsMin=<0 -15 -15>, angularTarget=<1 0 0>}
      localJointOffset=<0.01 0 0>
     }]
    relativeSpaceBone {boneName=head}
    simSpaceSettings {maxAngularAcceleration=5000, maxAngularVelocity=400, simSpaceAngularAlpha=0.2}
   }
   Output Pose(Result=Component To Local(ComponentPose=@n5)) {layerGroup=None}
  ]}
 EventGraph {nodes [
   n1 = Event OnCharacterCustomizationLoaded()
   > Set Body_Mesh(Body_Mesh=Get Array Item with EFort Custom Part Type(Array=@n1.CustomMeshParts))
   Event BlueprintUpdateAnimation()
  ]}
]
variables [Body_Mesh {type=SkeletalMeshComponent}]
